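# Unlit translucent effect material for the StreetDemon backpack, with animated color, opacity and vertex motion.
# A normal-perturbed view-facing term drives a color curve, optionally tinted by an HDR flame curve, with panning noise or circuit textures masking opacity.
# Scrolling noise lines, optional stepped flicker and an emissive mask refine the result before CharacterDeRez_Transition.
# World position offset adds sine wiggle and noise displacement steered by vertex color.
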Material /BRCosmetics/Accessories/FORT_Backpacks/M_MED_StreetDemon_Backpack/FX/M_FX
graph {
 properties {
  bEnableResponsiveAA=true
  bUseMaterialAttributes=true
  bUsedWithMeshDeformer=true
  bUsedWithMeshParticles=true
  bUsedWithSkeletalMesh=true
  blendMode=BLEND_Translucent
  pixelDepthOffsetMode=PDOM_Legacy
  shadingModel=MSM_Unlit
  usedShadingModels=Unlit
 }
 nodes [
  n1 = TextureCoordinate()
  n2 = VectorParameter() {parameterName=UV, defaultValue=<rgba 0.5 0.5 0.05 0.0679>}
  n3 = ComponentMask(None=@n2.RGB) {g=true, r=true}
  n4 = TextureSampleParameter2D(UVs=Panner(Coordinate=Multiply(A=@n1, B=@n3), Speed=AppendVector(A=@n2.B, B=@n2.B))) {parameterName=Param, samplerType=SAMPLERTYPE_Normal, texture=/Game/Global/Textures/Default/Tileable_Patterns/Cloud_normals}
  n5 = LinearInterpolate(A=Constant3Vector() {constant=<rgba 0 0 1 0>}, B=@n4.RGB, Alpha=ScalarParameter() {defaultValue=0.25, parameterName=NormalStrength})
  n6 = Saturate(None=DotProduct(A=CameraVectorWS(), B=Transform(None=@n5)))
  n7 = MaterialFunctionCall(XY=TextureCoordinate()) {materialFunction=&Swizzle}
  n8 = Divide(A=Distance(A=ObjectPositionWS(), B=CameraPositionWS())) {constB=256}
  n9 = Floor(None=LinearInterpolate(Alpha=Saturate(None=@n8)) {constA=18, constB=9})
  n10 = AppendVector(A=@n9, B=Multiply(A=@n9) {constB=1.5})
  n11 = CurveAtlasRowParameter(CurveTime=TextureSample(UVs=Multiply(A=@n7.YX, B=@n10)) {samplerType=SAMPLERTYPE_LinearColor, texture=/Game/Global/Textures/Default/Masks_and_Gradients/HexPattern}.R) {
   atlas=&AthenaCurveAtlas01
   curve=/Game/Accessories/FORT_Backpacks/M_MED_StreetDemon_Backpack/FX/Curve_Ramp
   defaultValue=0.3145
   parameterName=Param2
  }
  Add(A=@n6, B=@n11.G)
  n12 = TransformPosition(None=WorldPosition() {worldPositionShaderOffset=WPT_ExcludeAllShaderOffsets}.XYZ) {transformSourceType=TRANSFORMPOSSOURCE_World}
  n13 = Subtract(
   A=TransformPosition(None=ObjectPositionWS()) {transformSourceType=TRANSFORMPOSSOURCE_World}
   B=ScalarParameter() {defaultValue=5.5, parameterName=ObjectPosition}
  )
  Divide(A=ComponentMask(None=Subtract(A=@n12, B=@n13)) {b=true}, B=Constant() {r=20})
  n14 = Time()
  n15 = VertexColor()
  n16 = Multiply(A=@n15.R) {constB=2}
  n17 = Sine(None=Multiply(A=Add(A=@n14, B=@n16), B=ScalarParameter() {defaultValue=0.15, parameterName=WiggleFreq}))
  n18 = ScalarParameter() {defaultValue=2, parameterName=WiggleDistance}
  n19 = Multiply(A=Transform(None=Constant3Vector() {constant=<rgba 0 0 1 0>}) {transformSourceType=TRANSFORMSOURCE_Local}, B=@n18)
  n20 = VertexColor()
  n21 = Multiply(A=Time(), B=ScalarParameter() {defaultValue=0.05, parameterName=WPO_Speed})
  n22 = TextureSample(UVs=Divide(A=Add(A=@n15.R, B=@n21), B=Constant2Vector() {g=1, r=64})) {samplerType=SAMPLERTYPE_LinearColor, texture=/Game/Effects/Fort_Effects/Textures/Tile/LinerNoise64x64}
  n23 = Power(Base=@n22.RGB, Exp=ScalarParameter() {defaultValue=4, parameterName=WPO_MotionExp})
  n24 = VectorParameter() {parameterName=WPO_Rand_Size, defaultValue=<rgba 6 0.5 0.5 1.25>}
  n25 = Multiply(A=@n23, B=@n24.A) {constB=2}
  n26 = Subtract(
   A=MaterialFunctionCall(X=ComponentMask(None=@n23) {r=true}, Y=@n22.G, Z=@n22.G) {materialFunction=&MakeFloat3}.Result
   B=Constant3Vector() {constant=<rgba 0 0.5 0.5 0>}
  )
  n27 = TransformPosition(None=WorldPosition().XYZ) {transformSourceType=TRANSFORMPOSSOURCE_World}
  n28 = Multiply(
   A=Transform(None=Multiply(A=@n26, B=@n24.RGB)) {transformSourceType=TRANSFORMSOURCE_Local}
   B=Normalize(VectorInput=ComponentMask(None=@n27) {r=true})
  )
  n29 = Add(A=Clamp(None=Multiply(A=@n17, B=@n19), Max=@n18) {maxDefault=5}, B=Clamp(None=If(A=@n20.G, "A > B"=@n25, "A < B"=@n28) {constB=0.5}))
  n30 = CurveAtlasRowParameter(CurveTime=@n6) {
   atlas=&AthenaCurveAtlas01
   curve=/Game/Accessories/FORT_Backpacks/M_MED_StreetDemon_Backpack/FX/Curve_Color
   defaultValue=81
  }
  n31 = ComponentMask(None=TextureCoordinate() {coordinateIndex=1}) {g=true}
  n32 = CurveAtlasRowParameter(CurveTime=Power(Base=OneMinus(None=@n31))) {
   atlas=&AthenaCurveAtlas01
   curve=/Game/Athena/LinearColorCurves/HDRFlamesCurve
   defaultValue=60
   parameterName=Param_1
  }
  n33 = VectorParameter() {
   group=Emissive Mask
   parameterName=Emissive Mask Color/Alpha
   sortPriority=3
   channelNames {a=Alpha Strength, b="", g="", r=""}
   defaultValue=<rgba 0 0 0 5>
  }
  n34 = TextureSampleParameter2D() {
   group=Emissive Mask
   parameterName=Emissive Mask Texture
   sortPriority=1
   texture=/BRCosmetics/Accessories/FORT_Backpacks/F_MED_Graffiti_Future_Backpack/Skins/Orange/Textures/T_F_MED_GraffitiFuture_Orange_backpack_FX
  }
  n35 = ChannelMaskParameter(None=@n34.RGBA) {group=Emissive Mask, maskChannel=Green, parameterName=Emissive Mask Channel, sortPriority=2, defaultValue=<rgba 0 1 0 0>}
  n36 = ScalarParameter() {group=Emissive Mask, parameterName=Use Emissive Mask, sortPriority=0}
  n37 = Multiply(A=@n30, B=LinearInterpolate(A=@n32, B=LinearInterpolate(A=@n32, B=@n33, Alpha=@n35.R), Alpha=@n36))
  n38 = ScalarParameter() {parameterName=Use_WroldPositionMask}
  n39 = TextureCoordinate() {coordinateIndex=1}
  n40 = VectorParameter() {parameterName=UV2, defaultValue=<rgba 0.5 1.5 0.1 0.206>}
  n41 = Multiply(A=ComponentMask(None=@n40.RGB) {g=true, r=true}, B=LinearInterpolate(Alpha=VertexColor().G) {constA=1, constB=0.309})
  n42 = Add(
   A=Panner(Coordinate=Multiply(A=@n39, B=@n41), Speed=AppendVector(A=@n40.B, B=@n40.B))
   B=Multiply(A=ComponentMask(None=@n4.RGB) {g=true, r=true}, B=ScalarParameter() {defaultValue=0.02, parameterName=NormalAffectMap})
  )
  n43 = TextureCoordinate() {coordinateIndex=1}
  n44 = ComponentMask(None=@n43) {r=true}
  n45 = AppendVector(A=Multiply(A=@n44, B=@n44), B=ComponentMask(None=@n43) {g=true})
  n46 = TextureSample(UVs=Panner(Coordinate=MaterialFunctionCall(XY=@n45) {materialFunction=&Swizzle}.YX) {speedX=1}) {texture=/Game/Effects/Fort_Effects/Textures/T_FX_Circuitboard2}
  n47 = Constant() {r=0.3}
  n48 = LinearInterpolate(Alpha=LinearInterpolate(
    A=TextureSample(UVs=@n42) {texture=&T_NoisePack_01}.R
    B=LinearInterpolate(A=@n46.G, B=@n46.G, Alpha=@n47)
    Alpha=ScalarParameter() {parameterName=MaskPoison}
   )) {constA=0.25}
  n49 = GetMaterialAttributes(None=SetMaterialAttributes!(
    MaterialAttributes=SetMaterialAttributes!(MaterialAttributes, World Position Offset=@n29) {attributeSetTypes [#1]}
    Emissive Color=LinearInterpolate(A=@n30, B=@n37, Alpha=@n38)
    Opacity=Multiply(A=@n30.A, B=@n48)
   ) {attributeSetTypes [#2, #3]}) {attributeGetTypes [#3, #2]}
  n50 = Multiply(A=@n49.Opacity, B=@n11.R)
  FeatureLevelSwitch(Default=@n50, ES2=@n49.Opacity, ES3_1=@n49.Opacity, SM4_REMOVED=@n50)
  n51 = WorldPosition()
  n52 = LinearInterpolate(
   A=ComponentMask(None=TransformPosition(None=@n51.XYZ) {transformSourceType=TRANSFORMPOSSOURCE_World}) {b=true}
   B=ComponentMask(None=TextureCoordinate()) {g=true}
   Alpha=ScalarParameter() {parameterName=MaskPoison}
  )
  n53 = ScalarParameter() {defaultValue=12, parameterName=LineScale1}
  n54 = VectorParameter() {parameterName=LineSpeed, defaultValue=<rgba 0.1 -0.1 0 0>}
  n55 = Divide(A=@n54.R) {constB=1000}
  n56 = TextureSample(UVs=Panner(Coordinate=Divide(A=@n52, B=@n53) {constB=32}, Speed=AppendVector(A=@n54.R, B=@n55))) {samplerType=SAMPLERTYPE_Normal, texture=/Game/Global/Textures/Default/Noise/64x64Noise}
  n57 = ComponentMask(None=TransformPosition(None=WorldPosition().XYZ) {transformSourceType=TRANSFORMPOSSOURCE_World}) {b=true}
  n58 = ComponentMask(None=TextureCoordinate()) {g=true}
  n59 = ScalarParameter() {parameterName=MaskPoison}
  n60 = AppendVector(
   A=ComponentMask(None=TextureCoordinate() {vTiling=0.5}) {r=true}
   B=Divide(A=LinearInterpolate(A=@n57, B=@n58, Alpha=@n59), B=ScalarParameter() {defaultValue=128, parameterName=FlickerScale}) {constB=32}
  )
  n61 = Ceil(None=Multiply(A=Time()) {constB=3})
  n62 = VectorParameter() {parameterName=Param_1, defaultValue=<rgba 0.1 -0.1 0 1>}
  n63 = Divide(A=@n62.R) {constB=1000}
  n64 = TextureSample(UVs=Panner(Coordinate=@n60, Time=@n61, Speed=AppendVector(A=@n62.R, B=@n63))) {samplerType=SAMPLERTYPE_LinearColor, texture=/Game/UI/Foundation/Textures/Pattern/T_CubePattern}
  n65 = VectorParameter() {defaultValue=<rgba 0.3 -0.1 0 1>}
  n66 = Divide(A=@n65.G) {constB=1000}
  n67 = TextureSample(UVs=Panner(Coordinate=@n60, Time=@n61, Speed=AppendVector(A=@n66, B=@n65.G))) {samplerType=SAMPLERTYPE_LinearColor, texture=&T_CubePattern}
  n68 = OneMinus(None=Round(None=Subtract(A=@n64.R, B=@n67.G)))
  n69 = Multiply(
   A=Multiply(A=@n49.Opacity, B=LinearInterpolate(Alpha=@n56.R) {constA=1, constAlpha=0.1, constB=0.55})
   B=LinearInterpolate(A=Constant() {r=1}, B=Clamp(None=@n68), Alpha=ScalarParameter() {parameterName=Flicker})
  )
  n70 = Panner(
   Coordinate=Divide(A=@n52, B=ScalarParameter() {defaultValue=128, parameterName=LineScale2}) {constB=32}
   Speed=AppendVector(A=@n54.G, B=Divide(A=@n54.G) {constB=10000000})
  )
  n71 = Multiply(A=Subtract(
    A=TextureSample(UVs=@n70) {samplerType=SAMPLERTYPE_Normal, texture=&64x64Noise}.G
    B=ScalarParameter() {defaultValue=0.9, parameterName=LineSub}
   ) {constB=0.99}) {constB=15}
  n72 = Multiply(A=@n49.Emissive Color, B=LinearInterpolate(Alpha=Saturate(None=@n71)) {constA=0.5, constB=2})
  n73 = SetMaterialAttributes!(
   MaterialAttributes=@n49.MaterialAttributes
   Opacity=LinearInterpolate(A=@n69, B=LinearInterpolate(A=@n69, B=@n33.Alpha Strength, Alpha=@n35.R), Alpha=@n36)
   Emissive Color=LinearInterpolate(A=@n72, B=LinearInterpolate(A=@n72, B=@n33, Alpha=@n35.R), Alpha=@n36)
  ) {attributeSetTypes [#3, #2]}
  MaterialFunctionCall!(Fresnel=ScalarParameter() {defaultValue=1, parameterName=Fresnel}, Material=@n73) {materialFunction=/Game/Packages/Fortress_SharedMaterials/Base_Material_Functions/CharacterDeRez_Transition}
  MaterialFunctionCall(In=@n73) {materialFunction=&MF_TechArtOverrides_accessories}
  RotateAboutAxis(NormalizedRotationAxis=Constant3Vector() {constant=<rgba 0 0 1 0>}, PivotPoint=ObjectPositionWS(), Position=WorldPosition().XYZ)
  ScalarParameter() {defaultValue=24, parameterName=HexScale}
 ]
}
